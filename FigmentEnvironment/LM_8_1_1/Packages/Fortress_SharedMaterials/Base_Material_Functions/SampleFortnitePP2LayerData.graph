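# Material function that samples Pivot Painter 2 layer data from two texture parameters at input UVs, likely for hierarchical wind animation.
# The position texture yields a decoded position, with alpha unpacked as the parent index; the X-axis texture yields a decoded axis vector and extent.
# Parent index, current index and texture resolution feed ReturnParentTextureInfo, which outputs parent UVs and an is-child flag for hierarchy lookups.
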MaterialFunction /FigmentEnvironment/LM_8_1_1/Packages/Fortress_SharedMaterials/Base_Material_Functions/SampleFortnitePP2LayerData
graph {nodes [
  n1 = FunctionInput() {inputName=UVs, inputType=FunctionInput_Vector2}
  n2 = TextureSampleParameter2D(UVs=@n1) {
   group=Wind Textures
   parameterName=Position and Index Texture
   samplerType=SAMPLERTYPE_LinearColor
   texture=/Engine/Functions/Engine_MaterialFunctions02/PivotPainter2/Black_1x1_EXR_Texture
  }
  n3 = MaterialFunctionCall!(Integer as Float=@n2.A) {materialFunction=/Engine/Functions/Engine_MaterialFunctions02/ms_PivotPainter2_UnpackIntegerAsFloat}
  FunctionOutput(None=@n3.Result) {outputName=Parent Index, sortPriority=1}
  n4 = FunctionInput() {inputName="Texture resolution (ParentUVs)", inputType=FunctionInput_Vector2, sortPriority=1, previewValue=<88 86 0 1>}
  n5 = MaterialFunctionCall!(
   "Parent Index As Float (See note)"=@n3.Result
   Texture Dimensions=@n4
   "Current Index (see note)"=FunctionInput() {inputName="Current Index (ParentUVs)", inputType=FunctionInput_Scalar}
  ) {materialFunction=/Engine/Functions/Engine_MaterialFunctions02/PivotPainter2/ms_PivotPainter2_ReturnParentTextureInfo}
  FunctionOutput(None=@n5.Parent UVs) {outputName=Parent UVs, sortPriority=2}
  FunctionOutput(None=@n5."Is Child? (See note)") {outputName=Is Child?, sortPriority=3}
  n6 = TextureSampleParameter2D(UVs=@n1) {
   group=Vertex Shader Textures
   parameterName=X-Axis Texture And X Extent Alpha
   samplerType=SAMPLERTYPE_LinearColor
   texture=/FigmentEnvironment/LM_8_1_1/Packages/Fortnite_UniqueMaterials_1/Textures/black_1x1_VectorDisplacement_Linear_Texture
  }
  FunctionOutput(None=AppendVector(
    A=MaterialFunctionCall!(Axis Vector RGB=@n6.RGB) {materialFunction=/Engine/Functions/Engine_MaterialFunctions02/PivotPainter2/ms_PivotPainter2_DecodeAxisVector}.Result
    B=MaterialFunctionCall!(8 Bit Alpha Extent Value=@n6.A) {materialFunction=/Engine/Functions/Engine_MaterialFunctions02/PivotPainter2/ms_PivotPainter2_Decode8BitAlphaAxisExtent}.Rescaled Extent
   )) {outputName=X axis vector and X extent}
  FunctionOutput(None=AppendVector(
    A=MaterialFunctionCall!(Position RGB=@n2.RGB) {materialFunction=/Engine/Functions/Engine_MaterialFunctions02/PivotPainter2/ms_PivotPainter2_DecodePostion}.Result
    B=@n3.Result
   )) {outputName=Position and Parent Index, sortPriority=-1}
  TextureObjectParameter() {
   group=Vertex Shader Textures
   parameterName=X-Axis Texture And X Extent Alpha
   samplerType=SAMPLERTYPE_LinearColor
   texture=&black_1x1_VectorDisplacement_Linear_Texture
  }
  TextureObjectParameter() {
   group=Vertex Shader Textures
   parameterName=Position and Index Texture
   samplerType=SAMPLERTYPE_LinearColor
   texture=&Black_1x1_EXR_Texture
  }
 ]}
