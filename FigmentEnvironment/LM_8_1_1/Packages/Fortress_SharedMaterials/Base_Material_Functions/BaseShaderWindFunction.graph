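# Material function for wind animation on Pivot Painter 2 meshes, outputting world position offset and a tangent-space normal.
# UV1 indexes data textures for each element's and its parent's pivot and axis.
# Two ball-and-socket constraint passes chain: level 2 rotates around the element's own pivot, then level 1 around the parent's for children, with code-controlled wind direction and offsets.
# UseWind off, ES2 and ES3_1 bypass it with zero offset and unchanged normal.
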MaterialFunction /FigmentEnvironment/LM_8_1_1/Packages/Fortress_SharedMaterials/Base_Material_Functions/BaseShaderWindFunction
graph {nodes [
  Add(B=WorldPosition().XYZ)
  Constant()
  n1 = TextureObjectParameter() {
   group=Wind Textures
   parameterName=Position and Index Texture
   samplerType=SAMPLERTYPE_LinearColor
   texture=/Engine/Functions/Engine_MaterialFunctions02/PivotPainter2/Black_1x1_EXR_Texture
  }
  n2 = TextureProperty(Tex=@n1)
  n3 = MaterialFunctionCall!(
   "Current Index (ParentUVs)"=MaterialFunctionCall!(Data Texture Dimensions=@n2, Pivot Painter UV Coordinates=TextureCoordinate() {coordinateIndex=1}) {materialFunction=/Engine/Functions/Engine_MaterialFunctions02/PivotPainter2/ms_PivotPainter2_CalculateMeshElementIndex}.Index
   UVs=TextureCoordinate() {coordinateIndex=1}
   "Texture resolution (ParentUVs)"=@n2
  ) {materialFunction=&/SampleFortnitePP2LayerData}
  n4 = ComponentMask(None=@n3.Position and Parent Index) {b=true, g=true, r=true}
  n5 = FunctionInput() {bUsePreviewValueAsDefault=true, inputName=Tangent Space Normal, previewValue=<0 0 1 1>}
  n6 = WorldPosition() {worldPositionShaderOffset=WPT_ExcludeAllShaderOffsets}
  n7 = MaterialFunctionCall!(
   Enable=StaticBoolParameter() {defaultValue=true, group=Wind Setting 2, parameterName=Animate Level 2}
   Enable Normal Rotation=StaticBoolParameter() {defaultValue=true, group=Wind Setting 2, parameterName=Animate Level 2 Normals}
   Position=@n4
   X-Axis Texture And X Extent Alpha=@n3.X axis vector and X extent
   UseWrapAroundLogic=StaticBoolParameter() {defaultValue=true, group=Wind Setting 2, parameterName=Wind 2 - Wrap Deformation Around Pivot Point}
   Wind Gust Angle Rotation=ScalarParameter() {defaultValue=0.1, group=Wind Setting 2, parameterName=Max Rotation 2, sliderMax=10}
   Wind Gust Offset=ScalarParameter() {group=Wind Setting 2, parameterName=Wind offset 2}
   Optional - world space pixel normal=Transform(None=@n5)
   Modulate Rotation Results by the Vertex Alpha=StaticBoolParameter() {group=Wind Setting 2, parameterName=Wind 2 - Modulate the results with the vertex alpha}
   Random Rotation Texture Sample Scale=ScalarParameter() {defaultValue=250, group=Wind Setting 2, parameterName=Wind 2 Random Rotation Change Rate}
   Random Rotation Influence=ScalarParameter() {defaultValue=0.25, group=Wind Setting 2, parameterName=Wind 2 Random Rotation Influence}
   World Position=@n6.XYZ
   Model Depth=@n3.X axis vector and X extent
   Effect's Target Depth=Constant() {r=1}
   TimeRate_CodeReferenced1=ScalarParameter() {defaultValue=0.1, group=Wind Setting 2, parameterName=Wind_2_Speed_CodeReferenced}
   TimeOffset_CodeControlled1=ScalarParameter() {group=Code Controlled, parameterName=Wind_2_Offset_CodeReferenced}
   Add Shear
   Max Sheer Offset
   Wind Direction=VectorParameter() {group=Code Controlled, parameterName=WindVector_CodeControlled, defaultValue=<rgba 1 0 0 1>}.RGB
  ) {materialFunction=&/Fornite_BallAndSocketConstraint}
  n8 = MaterialFunctionCall("Current Index (ParentUVs)"=@n3.Parent Index, UVs=@n3.Parent Index, "Texture resolution (ParentUVs)"=@n2) {materialFunction=&SampleFortnitePP2LayerData}
  n9 = Clamp(None=@n3.Is Child?)
  n10 = MaterialFunctionCall(
   Enable=StaticBoolParameter() {defaultValue=true, group=Wind Setting 1, parameterName=Animate Level 1}
   Enable Normal Rotation=StaticBoolParameter() {group=Wind Setting 1, parameterName=Animate Level 1 Normals}
   Position=LinearInterpolate(A=@n4, B=ComponentMask(None=@n8.Position and Parent Index) {b=true, g=true, r=true}, Alpha=@n9) {desc=location}
   X-Axis Texture And X Extent Alpha=LinearInterpolate(A=@n3.X axis vector and X extent, B=@n8.X axis vector and X extent, Alpha=@n9) {desc=rotation}
   UseWrapAroundLogic=StaticBoolParameter() {group=Wind Setting 1, parameterName=Wind 1 - Wrap Deformation Around Pivot Point}
   Wind Gust Angle Rotation=ScalarParameter() {defaultValue=0.1, group=Wind Setting 1, parameterName=Max Rotation 1, sliderMax=10}
   Wind Gust Offset=ScalarParameter() {group=Wind Setting 1, parameterName=Wind offset 1}
   Optional - world space pixel normal=@n7.Rotated World Space Pixel Normal
   Modulate Rotation Results by the Vertex Alpha=StaticBoolParameter() {group=Wind Setting 1, parameterName=Wind 1 - Modulate the results with the vertex alpha}
   Random Rotation Texture Sample Scale=ScalarParameter() {defaultValue=2.5, group=Wind Setting 1, parameterName=Wind 1 Random Rotation Change Rate}
   Random Rotation Influence=ScalarParameter() {defaultValue=1, group=Wind Setting 1, parameterName=Wind 1 Random Rotation Influence}
   World Position=Add(A=@n7.Add to WPO, B=@n6.XYZ)
   Model Depth=@n3.Is Child?
   Effect's Target Depth=Constant()
   TimeRate_CodeReferenced1=ScalarParameter() {defaultValue=0.1, group=Wind Setting 1, parameterName=Wind_1_Speed_CodeReferenced}
   TimeOffset_CodeControlled1=ScalarParameter() {group=Code Controlled, parameterName=Wind_1_Offset_CodeReferenced}
   Wind Direction=VectorParameter() {group=Code Controlled, parameterName=WindVector_CodeControlled, defaultValue=<rgba 1 0 0 1>}.RGB
  ) {materialFunction=&Fornite_BallAndSocketConstraint}
  n11 = Constant()
  n12 = FunctionInput(Preview=StaticBool()) {inputName=UseWind, inputType=FunctionInput_StaticBool}
  n13 = StaticSwitch(True=Add(A=@n7.Add to WPO, B=@n10.Add to WPO), False=@n11, Value=@n12)
  FunctionOutput(None=FeatureLevelSwitch(Default=@n13, ES2=@n11, ES3_1=@n11, SM4_REMOVED=@n13)) {outputName=World position offset}
  n14 = StaticSwitch(
   True=Transform(None=@n10.Rotated World Space Pixel Normal) {transformSourceType=TRANSFORMSOURCE_World, transformType=TRANSFORM_Tangent}
   False=@n5
   Value=@n12
  )
  FunctionOutput(None=FeatureLevelSwitch(Default=@n14, ES2=@n5, ES3_1=@n5, SM4_REMOVED=@n14)) {outputName=Tangent Space Pixel Normal}
  ScalarParameter() {defaultValue=-2000, group=Wind Setting 2, parameterName=Wind 2 - Vertex Phase Offset X Axis}
  ScalarParameter() {defaultValue=1, group=Wind Setting 2, parameterName=Radius Multiplier 2}
  ScalarParameter() {defaultValue=1, group=Wind Setting 1, parameterName=Radius Multiplier 1}
  StaticSwitchParameter() {group=ParticleShader, parameterName=IsParticleShader?}
  TextureObjectParameter() {
   group=Wind
   parameterName=X-Axis Texture And X Extent Alpha
   samplerType=SAMPLERTYPE_LinearColor
   texture=/FigmentEnvironment/LM_8_1_1/Packages/Fortnite_UniqueMaterials_1/Textures/black_1x1_VectorDisplacement_Linear_Texture
  }
  TextureObjectParameter() {
   group=Wind
   parameterName=Position and Index Texture
   samplerType=SAMPLERTYPE_LinearColor
   texture=&Black_1x1_EXR_Texture
  }
 ]}
