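# Cloth material function that animates flag-like motion, outputting a normal and an additive world position offset.
# Tiled, rotated UVs pan over time through flag textures, desynchronized by actor position or optional stepped UV1 offsets.
# The mask texture drives displacement along a local-space direction, optionally blended with the vertex normal.
# A vertex-color mask scales displacement and normal intensity, with settings exposed as parameters and overridable inputs.
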MaterialFunction /FigmentEnvironment/LM_8_1_1/Packages/Fortress_SharedMaterials/Base_Material_Functions/Cloth
graph {
 properties {description=Parameters are held within. Adjust the cloth settings in a material instance.}
 nodes [
  n1 = ScalarParameter() {defaultValue=1, group=WPO_ClothAnimation, parameterName=WPO_ClothTiling_Y}
  n2 = ScalarParameter() {defaultValue=1, group=WPO_ClothAnimation, parameterName=WPO_ClothTiling}
  n3 = Multiply(
   A=TextureCoordinate()
   B=StaticSwitchParameter(True=AppendVector(A=@n1, B=@n2), False=@n1) {group=WPO_ClothAnimation, parameterName=UseClothNonUniformTiling}
  )
  n4 = ScalarParameter() {defaultValue=1, group=WPO_ClothAnimation, parameterName=WPO_ClothPanDirection}
  n5 = Multiply(
   A=Time()
   B=FunctionInput(Preview=ScalarParameter() {defaultValue=1, group=WPO_ClothAnimation, parameterName=WPO_ClothPanSpeed}) {bUsePreviewValueAsDefault=true, inputName=Pan Speed, inputType=FunctionInput_Scalar}
  )
  n6 = Multiply(
   A=ComponentMask(None=TextureCoordinate() {coordinateIndex=1}) {r=true}
   B=ScalarParameter() {defaultValue=12, group=WPO_ClothAnimation, parameterName=ClothRandomOffsetSteps}
  )
  n7 = Add(
   A=@n5
   B=Multiply(A=Floor(None=@n6), B=ScalarParameter() {defaultValue=44.8109, group=WPO_ClothAnimation, parameterName=ClothRandomOffsetAmount})
  )
  n8 = Divide(A=ActorPositionWS(), B=ScalarParameter() {defaultValue=16.71, group=WPO_ClothAnimation, parameterName="ClothAnimOffsetChange(WS)"})
  n9 = Add(A=Add(A=ComponentMask(None=@n8) {r=true}, B=ComponentMask(None=@n8) {g=true}), B=@n5)
  n10 = Add(
   A=Rotator(
    Coordinate=FunctionInput(Preview=@n3) {bUsePreviewValueAsDefault=true, inputName=UVs, inputType=FunctionInput_Vector2}
    Time=FunctionInput(Preview=@n4) {bUsePreviewValueAsDefault=true, inputName=Pan Direction, inputType=FunctionInput_Scalar}
   )
   B=Multiply(
    A=Constant2Vector() {g=0.33, r=0.1}
    B=StaticSwitchParameter(True=@n7, False=@n9) {group=WPO_ClothAnimation, parameterName=UseClothRandomOffsetUV2}
   )
  )
  n11 = FunctionInput(Preview=ScalarParameter() {defaultValue=1, group=WPO_ClothAnimation, parameterName=WPO_ClothNormalIntensity}) {bUsePreviewValueAsDefault=true, inputName=WPO_ClothNormalIntensity, inputType=FunctionInput_Scalar}
  n12 = VertexColor()
  n13 = FunctionInput(Preview=ChannelMaskParameter(None=AppendVector(A=@n12, B=@n12)) {group=WPO_ClothAnimation, parameterName=ClothMask}.R) {bUsePreviewValueAsDefault=true, inputName="Motion Mask (Defaults to the Red Vertex Color)", inputType=FunctionInput_Scalar}
  FunctionOutput(None=LinearInterpolate(
    A=Constant3Vector() {constant=<rgba 0 0 1 0>}
    B=TextureSample(UVs=@n10) {samplerType=SAMPLERTYPE_Normal, texture=/FigmentEnvironment/LM_8_1_1/Environments/Sets/Junkyard/Textures/T_Flag_N}.RGB
    Alpha=Multiply(A=@n11, B=@n13)
   )) {outputName=Normal}
  n14 = TextureSample(UVs=@n10) {texture=/FigmentEnvironment/LM_8_1_1/Environments/Sets/Junkyard/Textures/T_Flag_M}
  n15 = FunctionInput(Preview=ScalarParameter() {defaultValue=10, group=WPO_ClothAnimation, parameterName=WPO_ClothOffsetIntensity}) {bUsePreviewValueAsDefault=true, inputName=WPO_ClothOffsetIntensity, inputType=FunctionInput_Scalar}
  n16 = Multiply(A=Multiply(A=@n14.R, B=@n15), B=@n13)
  n17 = Transform(None=FunctionInput(Preview=VectorParameter() {group=WPO_ClothAnimation, parameterName=WPO_ClothOffsetDirection, defaultValue=<rgba 1 0 0 1>}.RGB) {bUsePreviewValueAsDefault=true, inputName=WPO_ClothOffsetDirection}) {transformSourceType=TRANSFORMSOURCE_Local}
  n18 = Add(A=Multiply(A=@n17, B=Subtract(A=@n14.G) {constB=0.5}), B=VertexNormalWS())
  n19 = FunctionInput(Preview=ScalarParameter() {group=WPO_ClothAnimation, parameterName=WPO_ClothWPOOffsetInWindDirection}) {bUsePreviewValueAsDefault=true, inputName=WPO_ClothWPOWindTextureOffset, inputType=FunctionInput_Scalar}
  n20 = Multiply(A=Multiply(A=Add(A=@n14.R, B=@n19), B=@n17), B=@n15)
  FunctionOutput(None=StaticSwitchParameter(True=Multiply(A=@n16, B=@n18), False=Multiply(A=@n20, B=@n13)) {group=WPO_ClothAnimation, parameterName=UseClothVertexNormal}) {outputName=Additive WPO}
 ]
}
