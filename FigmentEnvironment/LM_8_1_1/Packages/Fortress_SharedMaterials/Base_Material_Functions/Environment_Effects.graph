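# Material function that layers optional "Skin FX" environment effects onto incoming material attributes, as part of the Fortress shared base material functions.
# Static switches chain three stages in order: a panning emissive blend, a thin film function, then an Ice-based crystal effect, each passing its input through when disabled.
# An effects mask texture, defaulting to all black, feeds static component mask parameters that choose channels for the emissive blend, crystal edge map and crystal alpha.
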MaterialFunction /FigmentEnvironment/LM_8_1_1/Packages/Fortress_SharedMaterials/Base_Material_Functions/Environment_Effects
graph {nodes [
  n1 = FunctionInput() {inputName=Environment FX, inputType=FunctionInput_MaterialAttributes}
  n2 = TextureSampleParameter2D() {group=Skin FX, parameterName=Environment Effects Mask, texture=/FigmentEnvironment/LM_8_1_1/Characters/Fortress_Character/Textures/AllBlack}
  n3 = AppendVector(A=@n2.RGB, B=@n2.RGB)
  n4 = StaticSwitchParameter(
   True=BlendMaterialAttributes(
    A=@n1
    B=MaterialFunctionCall!(Material=@n1) {materialFunction=&/Environment_PanningEmissive}.Result
    Alpha=StaticComponentMaskParameter(None=@n3) {defaultR=true, group=Skin FX, parameterName=Panning Emissive Channel}
   )
   False=@n1
  ) {group=Skin FX, parameterName=UsePanning Emissive}
  n5 = StaticSwitchParameter(
   True=MaterialFunctionCall!(Alpha, Material Attributes=@n4) {materialFunction=/FigmentEnvironment/LM_8_1_1/Weapons/WeaponSkins/MaterialFunction/ThinFilmFunction}.Result
   False=@n4
  ) {group=Skin FX, parameterName=Use Thin Film}
  n6 = StaticComponentMaskParameter(None=@n3) {defaultR=true, group=Skin FX, parameterName=Crystal Edge Map Channel}
  n7 = StaticComponentMaskParameter(None=@n3) {defaultR=true, group=Skin FX, parameterName=Crystal Channel}
  FunctionOutput(None=StaticSwitchParameter(
    True=MaterialFunctionCall!(Material=@n5, EdgeMap=@n6, Alpha=@n7) {materialFunction=/FigmentEnvironment/LM_8_1_1/Weapons/WeaponSkins/MaterialFunction/Ice}.Result
    False=@n5
   ) {group=Skin FX, parameterName=Use Crystal})
 ]}
