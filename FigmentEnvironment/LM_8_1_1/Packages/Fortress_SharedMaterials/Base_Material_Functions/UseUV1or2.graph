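# UseUV1or2 is a small material function in the Fortress shared base functions that picks which UV channel a material uses.
# A static bool input, "Use UV 2", drives a static switch that outputs the second UV channel (index 1) when true and the default first channel otherwise.
# The static switch resolves at compile time, and the input's preview bool serves as its default.
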MaterialFunction /FigmentEnvironment/LM_8_1_1/Packages/Fortress_SharedMaterials/Base_Material_Functions/UseUV1or2
graph {
 properties {description=PickUVChannel, libraryCategoriesText [Misc,Fortnite]}
 nodes [
  n1 = StaticBool()
  FunctionOutput(None=StaticSwitch(
    True=TextureCoordinate() {coordinateIndex=1}
    False=TextureCoordinate()
    Value=FunctionInput(Preview=@n1) {bUsePreviewValueAsDefault=true, inputName=Use UV 2, inputType=FunctionInput_StaticBool}
   ))
 ]
}
