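# Material function that tints a diffuse color with a world-space vertical gradient, for selected texture groups.
# Pixel height relative to the actor, with offset and length parameters, samples a gradient texture that is multiplied or added (switch named Screen) onto the diffuse.
# A dot product of material masks and a group vector, scaled by a parameter, drives the blend.
# A static switch and ES2/ES3_1 feature levels bypass the effect.
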MaterialFunction /FigmentEnvironment/LM_8_1_1/Packages/Fortress_SharedMaterials/Base_Material_Functions/WorldSpaceGradientFunction
graph {nodes [
  n1 = FunctionInput() {inputName=Diffuse, inputType=FunctionInput_Vector4, previewValue=<1 1 1 1>}
  n2 = ComponentMask(None=@n1) {b=true, g=true, r=true}
  n3 = ScalarParameter() {group=z_code, parameterName=Code_WSGradientParametersSet}
  n4 = VectorParameter() {group=z_code, parameterName=CODE_TextureGroupsToApplyGradientsTo, defaultValue=<rgba 1 0 1 0>}
  n5 = DotProduct(A=FunctionInput() {inputName=Materials masked per float 4 channel, inputType=FunctionInput_Vector4}, B=AppendVector(A=@n4.RGB, B=@n4.RGB))
  n6 = WorldPosition() {worldPositionShaderOffset=WPT_ExcludeAllShaderOffsets}
  n7 = ActorPositionWS()
  n8 = Subtract(
   A=Subtract(A=ComponentMask(None=@n6.XYZ) {b=true}, B=ComponentMask(None=@n7) {b=true})
   B=ScalarParameter() {group=z_code, parameterName=CODE_minGradientOffset}
  )
  n9 = ScalarParameter() {group=z_code, parameterName=CODE_gradientLength}
  n10 = TextureSampleParameter2D(UVs=Clamp(None=Divide(A=@n8, B=@n9))) {group=z_code, parameterName=WSGradient, texture=/FigmentEnvironment/LM_8_1_1/Packages/Fortress_SharedMaterials/Textures/WhiteTexture}
  n11 = StaticSwitchParameter(True=Add(A=@n2, B=@n10.RGB), False=Multiply(A=@n2, B=@n10.RGB)) {group=World Space Gradient, parameterName=Use GradientBlendModeScreen}
  n12 = ScalarParameter() {defaultValue=1, group=z_code, parameterName=CODE_GradientModulation}
  n13 = AppendVector(A=LinearInterpolate(A=@n2, B=Multiply(A=@n11, B=@n12), Alpha=Multiply(A=@n3, B=@n5)), B=ComponentMask(None=@n1) {a=true})
  FunctionOutput(None=StaticSwitchParameter(True=FeatureLevelSwitch(Default=@n13, ES2=@n1, ES3_1=@n1, SM4_REMOVED=@n13), False=@n1) {group=World Space Gradient, parameterName=Use WS Gradient}) {outputName=Out}
  n14 = ComponentMask(None=TextureCoordinate() {coordinateIndex=1}) {r=true}
  MaterialFunctionCall(
   X=SphereMask(A=@n14, B=Constant() {r=0.5}) {attenuationRadius=0.5}
   Y=SphereMask(A=@n14, B=Constant() {r=1.5}) {attenuationRadius=0.5}
   Z=SphereMask(A=@n14, B=Constant() {r=2.5}) {attenuationRadius=0.5}
   A=SphereMask(A=@n14, B=Constant() {r=3.5}) {attenuationRadius=0.5}
  ) {materialFunction=&MakeFloat4}
  StaticSwitch(
   True=@n13
   False=@n1
   Value=FunctionInput(Preview=StaticBool()) {bUsePreviewValueAsDefault=true, inputName=Use Gradient, inputType=FunctionInput_StaticBool, sortPriority=-2}
  )
 ]}
